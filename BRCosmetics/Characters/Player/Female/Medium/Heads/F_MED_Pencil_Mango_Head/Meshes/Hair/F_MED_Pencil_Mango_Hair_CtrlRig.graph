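# Control Rig for the F_MED_Pencil_Mango head's hair that offsets master hair bones according to head orientation.
# A Spherical Pose Reader on the head, one per branch of the Is Crouching bool, outputs the weight for the offsets.
# A library function adds translate, Euler rotate and scale deltas to each bone's current local transform.
# Standing offsets the back-bottom and one right-side bone, while crouching nudges only the back-bottom bone slightly.
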ControlRigBlueprint /BRCosmetics/Characters/Player/Female/Medium/Heads/F_MED_Pencil_Mango_Head/Meshes/Hair/F_MED_Pencil_Mango_Hair_CtrlRig
graphs [
 ContainedGraph {nodes [
   n1 = RigVMFunctionEntryNode()
   RigUnit_GetTransform = &RigUnit_GetTransform(Space=LocalSpace, bInitial=false, Item=@n1.Joint)
   n2 = &RigVMFunction_MathQuaternionToEuler(RotationOrder=ZYX, Value=@RigUnit_GetTransform.Transform.Rotation)
   > RigUnit_SetTransform = &RigUnit_SetTransform(
    Space=LocalSpace
    Value=(Rotation=(X=0.0132,Y=0.4099,Z=-0.0222,W=0.9118),Translation=(X=23.6045,Y=-6.37,Z=140.1115),Scale3D=(X=1,Y=1,Z=1))
    bInitial=false
    Item=@n1.Joint
    Weight=@n1.Weight
    Value.Translation.Z=&RigVMFunction_MathFloatAdd(B=@n1.Translate Z, A=@RigUnit_GetTransform.Transform.Translation.Z) {template="Add::Execute(in A,in B,out Result)"}.Result
    Value.Scale3D.Z=&RigVMFunction_MathFloatAdd(B=@n1.Scale Z, A=@RigUnit_GetTransform.Transform.Scale3D.Z) {template="Add::Execute(in A,in B,out Result)"}.Result
    Value.Translation.Y=&RigVMFunction_MathFloatAdd(B=@n1.Translate Y, A=@RigUnit_GetTransform.Transform.Translation.Y) {template="Add::Execute(in A,in B,out Result)"}.Result
    Value.Scale3D.Y=&RigVMFunction_MathFloatAdd(B=@n1.Scale Y, A=@RigUnit_GetTransform.Transform.Scale3D.Y) {template="Add::Execute(in A,in B,out Result)"}.Result
    Value.Translation.X=&RigVMFunction_MathFloatAdd(B=@n1.Translate X, A=@RigUnit_GetTransform.Transform.Translation.X) {template="Add::Execute(in A,in B,out Result)"}.Result
    Value.Scale3D.X=&RigVMFunction_MathFloatAdd(B=@n1.Scale X, A=@RigUnit_GetTransform.Transform.Scale3D.X) {template="Add::Execute(in A,in B,out Result)"}.Result
    Value.Rotation=&RigVMFunction_MathQuaternionFromEuler(
     Euler=(X=0,Y=0,Z=0)
     Euler.X=&RigVMFunction_MathFloatAdd(B=@n1.Rotate X, A=@n2.Result.X) {template="Add::Execute(in A,in B,out Result)"}.Result
     Euler.Y=&RigVMFunction_MathFloatAdd(B=@n1.Rotate Y, A=@n2.Result.Y) {template="Add::Execute(in A,in B,out Result)"}.Result
     Euler.Z=&RigVMFunction_MathFloatAdd(B=@n1.Rotate Z, A=@n2.Result.Z) {template="Add::Execute(in A,in B,out Result)"}.Result
    ).Result
   )
   > RigVMFunctionReturnNode()
  ]}
 RigVMFunctionLibrary {nodes [Set Transforms for Bone = RigVMCollapseNode(
    ExecuteContext=()
    Joint=()
    Rotate X=0
    Rotate Y=0
    Rotate Z=0
    Scale X=0
    Scale Y=0
    Scale Z=0
    Translate X=0
    Translate Y=0
    Translate Z=0
    Weight=1
   )]}
 RigVMModel {nodes [
   RigUnit_BeginExecution = &RigUnit_BeginExecution()
   Branch_1_RigVMUnitNode = &RigVMFunction_ControlFlowBranch(ExecuteContext=@RigUnit_BeginExecution.ExecutePin, Condition=RigVMVariableNode() {variable=Is Crouching}.Value)
   n1 = &RigUnit_SphericalPoseReader(
    ActiveRegionScaleFactors=(PositiveWidth=1,NegativeWidth=1,PositiveHeight=0.5,NegativeHeight=1)
    ActiveRegionSize=0.1
    Debug=(bDrawDebug=True,bDraw2D=False,bDrawLocalAxes=False,DebugScale=25,DebugSegments=20,DebugThickness=0.25)
    DriverAxis=(X=1,Y=0,Z=0)
    DriverItem=(Type=Bone,Name="head")
    FalloffRegionScaleFactors=(PositiveWidth=1,NegativeWidth=1,PositiveHeight=1.8,NegativeHeight=1)
    FalloffSize=0.18
    FlipHeightScaling=false
    FlipWidthScaling=false
    OptionalParentItem=(Type=Bone)
    RotationOffset=(X=-34,Y=-90,Z=0)
    ExecutePin=@Branch_1_RigVMUnitNode.False
   )
   > &RigUnit_ModifyTransforms(
    ItemToModify=((Item=(Type=Bone,Name="dyn_back_bottom_MSTR"),Transform=(Rotation=(X=0,Y=0,Z=0,W=1),Translation=(X=0,Y=0,Z=0),Scale3D=(X=1,Y=1,Z=1))))
    ItemToModify.0.Transform.Rotation=&RigVMFunction_MathQuaternionFromEuler(Euler=(X=0,Y=0,Z=0)).Result
    Weight=@n1.OutputParam
   )
   > Set Transforms for Bone_1_1_1_1_1_1_1_1 = RigVMFunctionReferenceNode(
    Joint=(Type=Bone,Name="dyn_back_bottom_MSTR")
    Rotate X=20
    Rotate Y=10
    Rotate Z=30
    Scale X=0
    Scale Y=0
    Scale Z=0
    Translate X=3
    Translate Y=-3
    Translate Z=-3
    Weight=@n1.OutputParam
   )
   > Set Transforms for Bone_1_1_1_1_1_1_1_1_2 = RigVMFunctionReferenceNode(
    Joint=(Type=Bone,Name="dyn_right_hair_02_MSTR")
    Rotate X=20
    Rotate Y=10
    Rotate Z=30
    Scale X=0
    Scale Y=0
    Scale Z=0
    Translate X=0
    Translate Y=0
    Translate Z=0
    Weight=@n1.OutputParam
   )
   n2 = &RigUnit_SphericalPoseReader(
    ActiveRegionScaleFactors=(PositiveWidth=1,NegativeWidth=1,PositiveHeight=0.5,NegativeHeight=1)
    ActiveRegionSize=0.1
    Debug=(bDrawDebug=True,bDraw2D=False,bDrawLocalAxes=False,DebugScale=25,DebugSegments=20,DebugThickness=0.25)
    DriverAxis=(X=1,Y=0,Z=0)
    DriverItem=(Type=Bone,Name="head")
    FalloffRegionScaleFactors=(PositiveWidth=1,NegativeWidth=1,PositiveHeight=1.8,NegativeHeight=1)
    FalloffSize=0.18
    FlipHeightScaling=false
    FlipWidthScaling=false
    OptionalParentItem=(Type=Bone)
    RotationOffset=(X=-34,Y=-90,Z=0)
    ExecutePin=@Branch_1_RigVMUnitNode.True
   )
   > &RigUnit_ModifyTransforms(
    ItemToModify=((Item=(Type=Bone,Name="dyn_back_bottom_MSTR"),Transform=(Rotation=(X=0,Y=0,Z=0,W=1),Translation=(X=0,Y=0,Z=0),Scale3D=(X=1,Y=1,Z=1))))
    ItemToModify.0.Transform.Rotation=&RigVMFunction_MathQuaternionFromEuler(Euler=(X=0,Y=0,Z=0)).Result
    Weight=@n2.OutputParam
   )
   > Set Transforms for Bone_1_1_1_1_1_1_1_1_1 = RigVMFunctionReferenceNode(
    Joint=(Type=Bone,Name="dyn_back_bottom_MSTR")
    Rotate X=2
    Rotate Y=0
    Rotate Z=0
    Scale X=0
    Scale Y=0
    Scale Z=0
    Translate X=1
    Translate Y=-1
    Translate Z=0
    Weight=@n2.OutputParam
   )
  ]}
]
hierarchy [
 root {loc=<0 0 0>, rot=<0 0 0 1>}
 pelvis {parent=root, loc=<0 -0.5301 98.0571>, rot=<0.0127 0.707 -0.0127 -0.707>}
 spine_01 {parent=pelvis, loc=<6.1208 -0.9328 0>, rot=<0 0 0.034 -0.9994>}
 spine_02 {parent=spine_01, loc=<6.5996 -0.2095 0>, rot=<0 0 0.0238 -0.9997>}
 spine_03 {parent=spine_02, loc=<6.4914 -0.1237 0>, rot=<0 0 0.0149 -0.9999>}
 spine_04 {parent=spine_03, loc=<6.9598 0.0794 0>, rot=<0 0 -0.0132 -0.9999>}
 spine_05 {parent=spine_04, loc=<6.9987 0 0>, rot=<0 0 -0.0015 -1>}
 attach_cape {parent=spine_05, loc=<9.3216 -8.4501 0>, rot=<0.0616 -0.7044 -0.0616 -0.7044>}
 neck_01 {parent=spine_05, loc=<11.3138 0.1529 0>, rot=<0 0 -0.2003 -0.9797>}
 neck_02 {parent=neck_01, loc=<4.5627 0 0>, rot=<0 0 0.0504 -0.9987>}
 head {parent=neck_02, loc=<5.3914 0 0>, rot=<0 0 0.115 -0.9934>}
 dyn_front_bang_MSTR {parent=head, loc=<8.6129 7.9971 0>, rot=<0 0 1 0>}
 dyn_front_bang_01 {parent=dyn_front_bang_MSTR, loc=<0 0 0>, rot=<0 0 0 1>}
 dyn_bang_right_MSTR {parent=head, loc=<8.4619 4.6816 5.5544>, rot=<-0.2111 0.2111 -0.6749 -0.6749>}
 dyn_bang_right_01 {parent=dyn_bang_right_MSTR, loc=<0 0 0>, rot=<0 0 0 1>}
 dyn_bang_left_MSTR {parent=head, loc=<8.4184 5.2897 -4.5318>, rot=<0.3064 -0.3064 -0.6373 -0.6373>}
 dyn_bang_left_01 {parent=dyn_bang_left_MSTR, loc=<0 0 0>, rot=<0 0 0 1>}
 dyn_hair_right_MSTR {parent=head, loc=<4.6066 1.347 7.666>, rot=<-0.0108 0.0974 0.99 0.1013>}
 dyn_hair_right_01 {parent=dyn_hair_right_MSTR, loc=<0 0 0>, rot=<0 0 0 1>}
 dyn_hair_right_02 {parent=dyn_hair_right_01, loc=<3.9921 -0.0275 -0.5071>, rot=<-0.0013 -0.0886 0.0924 -0.9918>}
 dyn_hair_left_MSTR {parent=head, loc=<5.4655 1.675 -7.3793>, rot=<-0.1342 -0.987 -0.0829 0.0289>}
 dyn_hair_left_01 {parent=dyn_hair_left_MSTR, loc=<0 0 0>, rot=<0 0 0 1>}
 dyn_hair_left_02 {parent=dyn_hair_left_01, loc=<4.7102 0 0>, rot=<0 0 0 1>}
 dyn_top_hair_MSTR {parent=head, loc=<8.3188 0.3408 0>, rot=<0 0 0 1>}
 dyn_top_hair_01 {parent=dyn_top_hair_MSTR, loc=<3.2486 8.4129 -0.3791>, rot=<0 0 -0.6419 -0.7668>}
 dyn_top_hair_02 {parent=dyn_top_hair_MSTR, loc=<6.9329 3.5086 0.9461>, rot=<0.0444 -0.0428 -0.5558 -0.829>}
 dyn_top_hair_03 {parent=dyn_top_hair_MSTR, loc=<7.153 -2.2765 0.3462>, rot=<0.0482 0.0339 -0.1746 -0.9829>}
 dyn_top_hair_04 {parent=dyn_top_hair_MSTR, loc=<4.9274 -4.5167 -5.2537>, rot=<0.7339 -0.5494 -0.2503 -0.3112>}
 dyn_back_bottom_MSTR {parent=head, loc=<-3.5744 -5.6461 -0.4202>, rot=<0.0022 -0.0132 0.9863 -0.1647>}
 dyn_back_bottom_01 {parent=dyn_back_bottom_MSTR, loc=<0 0 0>, rot=<0 0 0 1>}
 dyn_back_hair_01_MSTR {parent=head, loc=<1.1097 -7.7684 0.4742>, rot=<0.1138 0.1984 0.7608 -0.6073>}
 dyn_back_hair_01_01 {parent=dyn_back_hair_01_MSTR, loc=<0 0 0>, rot=<0 0 0 1>}
 dyn_back_hair_02_MSTR {parent=head, loc=<-0.2846 -5.8777 3.6907>, rot=<0.1399 0.1757 0.7242 -0.652>}
 dyn_back_hair_02_01 {parent=dyn_back_hair_02_MSTR, loc=<0 0 0>, rot=<0 0 0 1>}
 dyn_right_hair_02_MSTR {parent=head, loc=<2.531 -4.9226 6.9831>, rot=<0.4542 0.4532 0.5812 -0.5006>}
 dyn_right_hair_02_01 {parent=dyn_right_hair_02_MSTR, loc=<0 0 0>, rot=<0 0 0 1>}
 dyn_right_hair_03_MSTR {parent=head, loc=<5.9843 -2.7437 7.6348>, rot=<0.4931 0.4865 0.5486 -0.4682>}
 dyn_right_hair_03_01 {parent=dyn_right_hair_03_MSTR, loc=<0 0 0>, rot=<0 0 0 1>}
 dyn_right_hair_04_MSTR {parent=head, loc=<4.8173 -8.0654 5.0288>, rot=<0.2344 0.3437 0.7643 -0.4927>}
 dyn_right_hair_04_01 {parent=dyn_right_hair_04_MSTR, loc=<0 0 0>, rot=<0 0 0 1>}
 dyn_left_left_01_MSTR {parent=head, loc=<2.7373 -7.8514 -4.0987>, rot=<-0.1128 -0.1282 0.7907 -0.5879>}
 dyn_left_hair_01 {parent=dyn_left_left_01_MSTR, loc=<0 0 0>, rot=<0 0 0 1>}
 dyn_left_hair_02_MSTR {parent=head, loc=<1.1546 -3.5485 -6.297>, rot=<-0.3858 -0.3968 0.7272 -0.4061>}
 dyn_left_hair_02_01 {parent=dyn_left_hair_02_MSTR, loc=<0 0 0>, rot=<0 0 0 1>}
 dyn_left_hair_03_MSTR {parent=head, loc=<6.6655 -2.4292 -8.5738>, rot=<-0.488 -0.3759 0.6263 -0.4777>}
 dyn_left_hair_03_01 {parent=dyn_left_hair_03_MSTR, loc=<0 0 0>, rot=<0 0 0 1>}
]
variables [Is Crouching {type=bool, default=False}]
