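# Physics asset for the hair of the F_MED_Pencil_Mango player head, defining capsule bodies for secondary motion of bangs, side, back and top hair.
# Each hair group has a MSTR root body joined to the head with free angular motion, and the head links down through two neck bodies.
# Strand bodies hang from these roots with angle limits: mostly single segments, two-segment chains for front bang and hair_left/right, and four parallel top pieces.
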PhysicsAsset /BRCosmetics/Characters/Player/Female/Medium/Heads/F_MED_Pencil_Mango_Head/Meshes/Hair/F_MED_Pencil_Mango_Hair_Physics
bodies {
 dyn_back_bottom_01 [{shapeType=Capsule, length=11.6013, radius=2.4558, shapeName=dyn_back_bottom_01_capsule, center=<2.1656 0.1506 0>}]
 dyn_back_bottom_MSTR [{shapeType=Capsule, length=1, radius=1, shapeName=dyn_back_bottom_MSTR_capsule}]
 dyn_back_hair_01_01 [{shapeType=Capsule, length=3, radius=3, shapeName=dyn_back_hair_01_01_capsule, center=<2.614 0 0>, rotation=<rot 90 0 0>}]
 dyn_back_hair_01_MSTR [{shapeType=Capsule, length=1, radius=1, shapeName=dyn_back_hair_01_MSTR_capsule}]
 dyn_back_hair_02_01 [{shapeType=Capsule, length=3, radius=3, shapeName=dyn_back_hair_02_01_capsule, center=<2.2497 0 0>, rotation=<rot 90 0 0>}]
 dyn_back_hair_02_MSTR [{shapeType=Capsule, length=1, radius=1, shapeName=dyn_back_hair_02_MSTR_capsule}]
 dyn_bang_left_01 [{shapeType=Capsule, length=2.7275, radius=2.2, shapeName=dyn_bang_left_01_capsule, center=<1.1818 3.4114 0>, rotation=<rot 90 -90 13.0357>}]
 dyn_bang_left_MSTR [{shapeType=Capsule, length=1, radius=1, shapeName=dyn_bang_left_MSTR_capsule}]
 dyn_bang_right_01 [{
   shapeType=Capsule
   length=2.7225
   radius=2.1
   shapeName=dyn_bang_left_01_capsule
   center=<1.1652 3.3397 0>
   rotation=<rot 90 -153.4349 -50.3992>
  }]
 dyn_bang_right_MSTR [{shapeType=Capsule, length=1, radius=1, shapeName=dyn_bang_right_MSTR_capsule}]
 dyn_front_bang_01 [{shapeType=Capsule, length=2.1484, radius=1.2, shapeName=dyn_hair_left_02_capsule, center=<2.134 0 0>, rotation=<rot 90 0 0>}]
 dyn_front_bang_02 [{shapeType=Capsule, length=3.1895, radius=1.3, shapeName=dyn_hair_left_02_capsule, center=<2.9 0 0>, rotation=<rot 90 0 0>}]
 dyn_front_bang_MSTR [{shapeType=Capsule, length=1, radius=1, shapeName=dyn_front_bang_MSTR_capsule}]
 dyn_hair_left_01 [{shapeType=Capsule, length=2.315, radius=0.7, shapeName=dyn_hair_left_01_capsule, center=<2.134 0 0>, rotation=<rot 90 0 0>}]
 dyn_hair_left_02 [{shapeType=Capsule, length=3.1895, radius=0.7, shapeName=dyn_hair_left_02_capsule, center=<2.134 0 0>, rotation=<rot 90 0 0>}]
 dyn_hair_left_MSTR [{shapeType=Capsule, length=1, radius=1, shapeName=dyn_hair_left_MSTR_capsule}]
 dyn_hair_right_01 [{shapeType=Capsule, length=2.315, radius=0.7, shapeName=dyn_hair_left_01_capsule, center=<2.134 0 0>, rotation=<rot 90 0 0>}]
 dyn_hair_right_02 [{shapeType=Capsule, length=3.1895, radius=0.7, shapeName=dyn_hair_left_02_capsule, center=<2.134 0 0>, rotation=<rot 90 0 0>}]
 dyn_hair_right_MSTR [{shapeType=Capsule, length=1, radius=1, shapeName=dyn_hair_right_MSTR_capsule}]
 dyn_left_hair_01 [{shapeType=Capsule, length=3, radius=3, shapeName=dyn_left_hair_01_capsule, center=<2.3088 0 0>, rotation=<rot 90 0 0>}]
 dyn_left_hair_02_01 [{shapeType=Capsule, length=3.6257, radius=1.8743, shapeName=dyn_left_hair_02_01_capsule, center=<3.0496 0 0>, rotation=<rot 90 0 0>}]
 dyn_left_hair_02_MSTR [{shapeType=Capsule, length=1, radius=1, shapeName=dyn_left_hair_02_MSTR_capsule}]
 dyn_left_hair_03_01 [{shapeType=Capsule, length=1.8046, radius=2.1705, shapeName=dyn_left_hair_03_01_capsule, center=<3.0143 0 0>, rotation=<rot 90 0 0>}]
 dyn_left_hair_03_MSTR [{shapeType=Capsule, length=1, radius=1, shapeName=dyn_left_hair_03_MSTR_capsule}]
 dyn_left_left_01_MSTR [{shapeType=Capsule, length=1, radius=1, shapeName=dyn_left_left_01_MSTR_capsule}]
 dyn_right_hair_02_01 [{shapeType=Capsule, length=3, radius=3, shapeName=dyn_right_hair_02_01_capsule, center=<2.7169 0 0>, rotation=<rot 90 0 0>}]
 dyn_right_hair_02_MSTR [{shapeType=Capsule, length=1, radius=1, shapeName=dyn_right_hair_02_MSTR_capsule}]
 dyn_right_hair_03_01 [{shapeType=Capsule, length=0.7702, radius=3, shapeName=dyn_right_hair_03_01_capsule, center=<2.3641 0 0>, rotation=<rot 90 0 0>}]
 dyn_right_hair_03_MSTR [{shapeType=Capsule, length=1, radius=1, shapeName=dyn_right_hair_03_MSTR_capsule}]
 dyn_right_hair_04_01 [{shapeType=Capsule, length=3, radius=3, shapeName=dyn_right_hair_04_01_capsule, center=<2.6527 0 0>, rotation=<rot 90 0 0>}]
 dyn_right_hair_04_MSTR [{shapeType=Capsule, length=1, radius=1, shapeName=dyn_right_hair_04_MSTR_capsule}]
 dyn_top_hair_01 [{shapeType=Capsule, length=0.4411, radius=2, shapeName=dyn_top_hair_01_capsule, center=<2.1292 0 0>, rotation=<rot 90 0 0>}]
 dyn_top_hair_02 [{shapeType=Capsule, length=0.18, radius=2, shapeName=dyn_top_hair_02_capsule, center=<1.9183 0 0>, rotation=<rot 90 0 0>}]
 dyn_top_hair_03 [{shapeType=Capsule, length=0.3929, radius=2, shapeName=dyn_top_hair_03_capsule, center=<2.0035 0 0>, rotation=<rot 90 0 0>}]
 dyn_top_hair_04 [{shapeType=Capsule, length=0, radius=2, shapeName=dyn_top_hair_04_capsule, center=<1.9109 0 0>, rotation=<rot 90 0 0>}]
 dyn_top_hair_MSTR [{shapeType=Capsule, length=1, radius=1, shapeName=dyn_top_hair_MSTR_capsule}]
 head [{shapeType=Capsule, length=3.313, radius=6.4375, center=<2.734 0.6891 0>, rotation=<rot 90 0 0>}]
 neck_01 [{shapeType=Capsule, length=14.9711, radius=3.695, shapeName=neck_01_capsule, center=<0 -1.4789 0>, rotation=<rot 0 -90 0>}]
 neck_02 [{shapeType=Capsule, length=2.55, radius=3.6054, shapeName=neck_02_capsule, center=<2.7073 0 0>, rotation=<rot 90 0 0>}]
}
constraints [
 {
  bone1Name=dyn_back_bottom_MSTR
  bone2Name=head
  swing1LimitDegrees=45
  swing1Motion=Free
  swing2LimitDegrees=45
  swing2Motion=Free
  twistLimitDegrees=45
  twistMotion=Free
 }
 {bone1Name=dyn_back_bottom_01, bone2Name=dyn_back_bottom_MSTR, swing1LimitDegrees=30, swing2LimitDegrees=5, twistLimitDegrees=0}
 {
  bone1Name=dyn_back_hair_01_MSTR
  bone2Name=head
  swing1LimitDegrees=45
  swing1Motion=Free
  swing2LimitDegrees=45
  swing2Motion=Free
  twistLimitDegrees=45
  twistMotion=Free
 }
 {bone1Name=dyn_back_hair_01_01, bone2Name=dyn_back_hair_01_MSTR, swing1LimitDegrees=45, swing2LimitDegrees=10, twistLimitDegrees=5}
 {
  bone1Name=dyn_back_hair_02_MSTR
  bone2Name=head
  swing1LimitDegrees=45
  swing1Motion=Free
  swing2LimitDegrees=45
  swing2Motion=Free
  twistLimitDegrees=45
  twistMotion=Free
 }
 {bone1Name=dyn_back_hair_02_01, bone2Name=dyn_back_hair_02_MSTR, swing1LimitDegrees=20, swing2LimitDegrees=10, twistLimitDegrees=5}
 {
  bone1Name=dyn_bang_left_MSTR
  bone2Name=head
  swing1LimitDegrees=45
  swing1Motion=Free
  swing2LimitDegrees=45
  swing2Motion=Free
  twistLimitDegrees=45
  twistMotion=Free
 }
 {bone1Name=dyn_bang_left_01, bone2Name=dyn_bang_left_MSTR, swing1LimitDegrees=25, swing2LimitDegrees=3, twistLimitDegrees=0}
 {
  bone1Name=dyn_bang_right_MSTR
  bone2Name=head
  swing1LimitDegrees=45
  swing1Motion=Free
  swing2LimitDegrees=45
  swing2Motion=Free
  twistLimitDegrees=45
  twistMotion=Free
 }
 {bone1Name=dyn_bang_right_01, bone2Name=dyn_bang_right_MSTR, swing1LimitDegrees=25, swing2LimitDegrees=3, twistLimitDegrees=0}
 {
  bone1Name=dyn_front_bang_MSTR
  bone2Name=head
  swing1LimitDegrees=45
  swing1Motion=Free
  swing2LimitDegrees=45
  swing2Motion=Free
  twistLimitDegrees=45
  twistMotion=Free
 }
 {bone1Name=dyn_front_bang_01, bone2Name=dyn_front_bang_MSTR, swing1LimitDegrees=15, swing2LimitDegrees=2, twistLimitDegrees=1}
 {
  bone1Name=dyn_hair_left_MSTR
  bone2Name=head
  swing1LimitDegrees=45
  swing1Motion=Free
  swing2LimitDegrees=45
  swing2Motion=Free
  twistLimitDegrees=45
  twistMotion=Free
 }
 {bone1Name=dyn_hair_left_01, bone2Name=dyn_hair_left_MSTR, swing1LimitDegrees=10, swing2LimitDegrees=10, twistLimitDegrees=10}
 {bone1Name=dyn_hair_left_02, bone2Name=dyn_hair_left_01, swing1LimitDegrees=18, swing2LimitDegrees=10, twistLimitDegrees=5}
 {
  bone1Name=dyn_hair_right_MSTR
  bone2Name=head
  swing1LimitDegrees=45
  swing1Motion=Free
  swing2LimitDegrees=45
  swing2Motion=Free
  twistLimitDegrees=45
  twistMotion=Free
 }
 {bone1Name=dyn_hair_right_01, bone2Name=dyn_hair_right_MSTR, swing1LimitDegrees=11, swing2LimitDegrees=5, twistLimitDegrees=5}
 {bone1Name=dyn_hair_right_02, bone2Name=dyn_hair_right_01, swing1LimitDegrees=11, swing2LimitDegrees=8, twistLimitDegrees=2}
 {
  bone1Name=dyn_left_hair_02_MSTR
  bone2Name=head
  swing1LimitDegrees=45
  swing1Motion=Free
  swing2LimitDegrees=45
  swing2Motion=Free
  twistLimitDegrees=45
  twistMotion=Free
 }
 {bone1Name=dyn_left_hair_02_01, bone2Name=dyn_left_hair_02_MSTR, swing1LimitDegrees=45, swing2LimitDegrees=20, twistLimitDegrees=20}
 {
  bone1Name=dyn_left_hair_03_MSTR
  bone2Name=head
  swing1LimitDegrees=45
  swing1Motion=Free
  swing2LimitDegrees=45
  swing2Motion=Free
  twistLimitDegrees=45
  twistMotion=Free
 }
 {bone1Name=dyn_left_hair_03_01, bone2Name=dyn_left_hair_03_MSTR, swing1LimitDegrees=45, swing2LimitDegrees=20, twistLimitDegrees=20}
 {
  bone1Name=dyn_left_left_01_MSTR
  bone2Name=head
  swing1LimitDegrees=45
  swing1Motion=Free
  swing2LimitDegrees=45
  swing2Motion=Free
  twistLimitDegrees=45
  twistMotion=Free
 }
 {bone1Name=dyn_left_hair_01, bone2Name=dyn_left_left_01_MSTR, swing1LimitDegrees=45, swing2LimitDegrees=10, twistLimitDegrees=5}
 {
  bone1Name=dyn_right_hair_02_MSTR
  bone2Name=head
  swing1LimitDegrees=45
  swing1Motion=Free
  swing2LimitDegrees=45
  swing2Motion=Free
  twistLimitDegrees=45
  twistMotion=Free
 }
 {bone1Name=dyn_right_hair_02_01, bone2Name=dyn_right_hair_02_MSTR, swing1LimitDegrees=45, swing2LimitDegrees=5, twistLimitDegrees=5}
 {
  bone1Name=dyn_right_hair_03_MSTR
  bone2Name=head
  swing1LimitDegrees=45
  swing1Motion=Free
  swing2LimitDegrees=45
  swing2Motion=Free
  twistLimitDegrees=45
  twistMotion=Free
 }
 {bone1Name=dyn_right_hair_03_01, bone2Name=dyn_right_hair_03_MSTR, swing1LimitDegrees=45, swing2LimitDegrees=10, twistLimitDegrees=5}
 {
  bone1Name=dyn_right_hair_04_MSTR
  bone2Name=head
  swing1LimitDegrees=45
  swing1Motion=Free
  swing2LimitDegrees=45
  swing2Motion=Free
  twistLimitDegrees=45
  twistMotion=Free
 }
 {bone1Name=dyn_right_hair_04_01, bone2Name=dyn_right_hair_04_MSTR, swing1LimitDegrees=45, swing2LimitDegrees=10, twistLimitDegrees=5}
 {
  bone1Name=dyn_top_hair_MSTR
  bone2Name=head
  swing1LimitDegrees=45
  swing1Motion=Free
  swing2LimitDegrees=45
  swing2Motion=Free
  twistLimitDegrees=45
  twistMotion=Free
 }
 {bone1Name=dyn_top_hair_01, bone2Name=dyn_top_hair_MSTR, swing1LimitDegrees=45, swing2LimitDegrees=10, twistLimitDegrees=5}
 {bone1Name=dyn_top_hair_02, bone2Name=dyn_top_hair_MSTR, swing1LimitDegrees=45, swing2LimitDegrees=10, twistLimitDegrees=5}
 {bone1Name=dyn_top_hair_03, bone2Name=dyn_top_hair_MSTR, swing1LimitDegrees=45, swing2LimitDegrees=10, twistLimitDegrees=5}
 {bone1Name=dyn_top_hair_04, bone2Name=dyn_top_hair_MSTR, swing1LimitDegrees=45, swing2LimitDegrees=10, twistLimitDegrees=5}
 {
  bone1Name=head
  bone2Name=neck_02
  swing1LimitDegrees=45
  swing1Motion=Free
  swing2LimitDegrees=45
  swing2Motion=Free
  twistLimitDegrees=45
  twistMotion=Free
 }
 {
  bone1Name=neck_02
  bone2Name=neck_01
  swing1LimitDegrees=45
  swing1Motion=Free
  swing2LimitDegrees=45
  swing2Motion=Free
  twistLimitDegrees=45
  twistMotion=Free
 }
 {bone1Name=dyn_front_bang_02, bone2Name=dyn_front_bang_01, swing1LimitDegrees=15, swing2LimitDegrees=30, twistLimitDegrees=1}
]
