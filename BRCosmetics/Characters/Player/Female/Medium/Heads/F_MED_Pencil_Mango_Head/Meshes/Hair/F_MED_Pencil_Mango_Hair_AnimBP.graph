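# Post-process style hair AnimBlueprint for the F_MED_Pencil_Mango head, adding secondary motion to hair bones.
# The live chain runs Input Pose through a Control Rig, a head-relative RigidBody simulation with custom gravity, then a small dyn_back_bottom_01 offset.
# An AnimDynamics chain, jump/fall blend and head-pose bone corrections sit disconnected from the output.
# The update event only sets Head_up true, leaving the Head Down calculation unexecuted.
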AnimBlueprint /BRCosmetics/Characters/Player/Female/Medium/Heads/F_MED_Pencil_Mango_Head/Meshes/Hair/F_MED_Pencil_Mango_Hair_AnimBP
target_skeleton=/BRCosmetics/Characters/Player/Common/Fortnite_Base_Head/Pencil/Fortnite_Base_Head_Pencil_Skeleton
graphs [
 AnimGraph {nodes [
   n1 = Anim Dynamics - Bone: dyn_back_bottom_01(Alpha=1, ExternalForce="0,0,0") {
    angularDampingOverride=0.9999
    angularSpringConstant=75
    bAngularSpring=true
    bGravityOverrideInSimSpace=true
    bOverrideAngularDamping=true
    simulationSpace=RootRelative
    alphaBoolBlend {blendOption=Linear}
    boundBone {boneName=dyn_back_bottom_01}
    componentLinearAccScale=<1 1 1>
    componentLinearVelScale=<1 1 1>
    physicsBodyDefinitions [{
      sphereCollisionRadius=0
      boundBone {boneName=dyn_back_bottom_01}
      boxExtents=<10 20 20>
      constraintSetup {angularLimitsMax=<0 0 5>, angularLimitsMin=<0 0 -5>, angularTarget=<0 0 0>}
      localJointOffset=<1 0 -0.2>
     }]
   }
   n2 = Anim Dynamics - Bone: dyn_front_bang_01(ComponentPose=@n1, Alpha=1, ExternalForce="0,0,0") {
    angularDampingOverride=0.9999
    angularSpringConstant=75
    bAngularSpring=true
    bGravityOverrideInSimSpace=true
    bOverrideAngularDamping=true
    simulationSpace=RootRelative
    alphaBoolBlend {blendOption=Linear}
    boundBone {boneName=dyn_front_bang_01}
    componentLinearAccScale=<1 1 1>
    componentLinearVelScale=<1 1 1>
    physicsBodyDefinitions [{
      sphereCollisionRadius=0
      boundBone {boneName=dyn_front_bang_01}
      boxExtents=<10 20 20>
      constraintSetup {angularLimitsMax=<5 1 2>, angularLimitsMin=<-5 -1 -2>, angularTarget=<0 0 0>}
      localJointOffset=<1 0 -0.2>
     }]
   }
   n3 = Anim Dynamics - Bone: dyn_hair_right_01(ComponentPose=@n2, Alpha=1, ExternalForce="0,0,0") {
    angularDampingOverride=0.9999
    angularSpringConstant=75
    bAngularSpring=true
    bGravityOverrideInSimSpace=true
    bOverrideAngularDamping=true
    simulationSpace=RootRelative
    alphaBoolBlend {blendOption=Linear}
    boundBone {boneName=dyn_hair_right_01}
    componentLinearAccScale=<1 1 1>
    componentLinearVelScale=<1 1 1>
    physicsBodyDefinitions [{
      sphereCollisionRadius=0
      boundBone {boneName=dyn_hair_right_01}
      boxExtents=<10 20 20>
      constraintSetup {angularLimitsMax=<0 5 15>, angularLimitsMin=<0 -5 -15>, angularTarget=<0 0 0>}
      localJointOffset=<1 0 -0.2>
     }]
   }
   n4 = Anim Dynamics - Bone: dyn_hair_right_02(ComponentPose=@n3, Alpha=1, ExternalForce="0,0,0") {
    angularDampingOverride=0.9999
    angularSpringConstant=75
    bAngularSpring=true
    bGravityOverrideInSimSpace=true
    bOverrideAngularDamping=true
    simulationSpace=RootRelative
    alphaBoolBlend {blendOption=Linear}
    boundBone {boneName=dyn_hair_right_02}
    componentLinearAccScale=<1 1 1>
    componentLinearVelScale=<1 1 1>
    physicsBodyDefinitions [{
      sphereCollisionRadius=0
      boundBone {boneName=dyn_hair_right_02}
      boxExtents=<10 20 20>
      constraintSetup {angularLimitsMax=<0 5 15>, angularLimitsMin=<0 -5 -15>, angularTarget=<0 0 0>}
      localJointOffset=<1 0 -0.2>
     }]
   }
   n5 = Anim Dynamics - Bone: dyn_hair_left_01(ComponentPose=@n4, Alpha=1, ExternalForce="0,0,0") {
    angularDampingOverride=0.9999
    angularSpringConstant=75
    bAngularSpring=true
    bGravityOverrideInSimSpace=true
    bOverrideAngularDamping=true
    simulationSpace=RootRelative
    alphaBoolBlend {blendOption=Linear}
    boundBone {boneName=dyn_hair_left_01}
    componentLinearAccScale=<1 1 1>
    componentLinearVelScale=<1 1 1>
    physicsBodyDefinitions [{
      sphereCollisionRadius=0
      boundBone {boneName=dyn_hair_left_01}
      boxExtents=<10 20 20>
      constraintSetup {angularLimitsMax=<0 5 15>, angularLimitsMin=<0 -5 -15>, angularTarget=<0 0 0>}
      localJointOffset=<1 0 -0.2>
     }]
   }
   n6 = Anim Dynamics - Bone: dyn_hair_left_02(ComponentPose=@n5, Alpha=1, ExternalForce="0,0,0") {
    angularDampingOverride=0.9999
    angularSpringConstant=75
    bAngularSpring=true
    bGravityOverrideInSimSpace=true
    bOverrideAngularDamping=true
    simulationSpace=RootRelative
    alphaBoolBlend {blendOption=Linear}
    boundBone {boneName=dyn_hair_left_02}
    componentLinearAccScale=<1 1 1>
    componentLinearVelScale=<1 1 1>
    physicsBodyDefinitions [{
      sphereCollisionRadius=0
      boundBone {boneName=dyn_hair_left_02}
      boxExtents=<10 20 20>
      constraintSetup {angularLimitsMax=<0 5 15>, angularLimitsMin=<0 -5 -15>, angularTarget=<0 0 0>}
      localJointOffset=<1 0 -0.2>
     }]
   }
   n7 = Anim Dynamics - Bone: dyn_bang_right_01(ComponentPose=@n6, Alpha=1, ExternalForce="0,0,0") {
    angularDampingOverride=0.9999
    angularSpringConstant=75
    bAngularSpring=true
    bGravityOverrideInSimSpace=true
    bOverrideAngularDamping=true
    simulationSpace=RootRelative
    alphaBoolBlend {blendOption=Linear}
    boundBone {boneName=dyn_bang_right_01}
    componentLinearAccScale=<1 1 1>
    componentLinearVelScale=<1 1 1>
    physicsBodyDefinitions [{
      sphereCollisionRadius=0
      boundBone {boneName=dyn_bang_right_01}
      boxExtents=<10 20 20>
      constraintSetup {angularLimitsMax=<3 1 5>, angularLimitsMin=<-3 -1 -5>, angularTarget=<0 0 0>}
      localJointOffset=<1 0 -0.2>
     }]
   }
   n8 = Anim Dynamics - Bone: dyn_bang_left_01(ComponentPose=@n7, Alpha=1, ExternalForce="0,0,0") {
    angularDampingOverride=0.9999
    angularSpringConstant=75
    bAngularSpring=true
    bGravityOverrideInSimSpace=true
    bOverrideAngularDamping=true
    simulationSpace=RootRelative
    alphaBoolBlend {blendOption=Linear}
    boundBone {boneName=dyn_bang_left_01}
    componentLinearAccScale=<1 1 1>
    componentLinearVelScale=<1 1 1>
    physicsBodyDefinitions [{
      sphereCollisionRadius=0
      boundBone {boneName=dyn_bang_left_01}
      boxExtents=<10 20 20>
      constraintSetup {angularLimitsMax=<3 1 5>, angularLimitsMin=<-3 -1 -5>, angularTarget=<0 0 0>}
      localJointOffset=<1 0 -0.2>
     }]
   }
   n9 = Anim Dynamics - Bone: dyn_left_hair_03_01(ComponentPose=@n8, Alpha=1, ExternalForce="0,0,0") {
    angularDampingOverride=0.9999
    angularSpringConstant=75
    bAngularSpring=true
    bGravityOverrideInSimSpace=true
    bOverrideAngularDamping=true
    simulationSpace=RootRelative
    alphaBoolBlend {blendOption=Linear}
    boundBone {boneName=dyn_left_hair_03_01}
    componentLinearAccScale=<1 1 1>
    componentLinearVelScale=<1 1 1>
    physicsBodyDefinitions [{
      sphereCollisionRadius=0
      boundBone {boneName=dyn_left_hair_03_01}
      boxExtents=<10 20 20>
      constraintSetup {angularLimitsMax=<0 10 1>, angularLimitsMin=<0 -10 -1>, angularTarget=<0 0 0>}
      localJointOffset=<1 0 -0.2>
     }]
   }
   n10 = Anim Dynamics - Bone: dyn_right_hair_03_01(ComponentPose=@n9, Alpha=1, ExternalForce="0,0,0") {
    angularDampingOverride=0.9999
    angularSpringConstant=75
    bAngularSpring=true
    bGravityOverrideInSimSpace=true
    bOverrideAngularDamping=true
    simulationSpace=RootRelative
    alphaBoolBlend {blendOption=Linear}
    boundBone {boneName=dyn_right_hair_03_01}
    componentLinearAccScale=<1 1 1>
    componentLinearVelScale=<1 1 1>
    physicsBodyDefinitions [{
      sphereCollisionRadius=0
      boundBone {boneName=dyn_right_hair_03_01}
      boxExtents=<10 20 20>
      constraintSetup {angularLimitsMax=<0 4 1>, angularLimitsMin=<0 -4 -1>, angularTarget=<0 0 0>}
      localJointOffset=<1 0 -0.2>
     }]
   }
   n11 = Anim Dynamics - Bone: dyn_back_hair_01_01(ComponentPose=@n10, Alpha=1, ExternalForce="0,0,0") {
    angularDampingOverride=0.9999
    angularSpringConstant=75
    bAngularSpring=true
    bGravityOverrideInSimSpace=true
    bOverrideAngularDamping=true
    simulationSpace=RootRelative
    alphaBoolBlend {blendOption=Linear}
    boundBone {boneName=dyn_back_hair_01_01}
    componentLinearAccScale=<1 1 1>
    componentLinearVelScale=<1 1 1>
    physicsBodyDefinitions [{
      sphereCollisionRadius=0
      boundBone {boneName=dyn_back_hair_01_01}
      boxExtents=<10 20 20>
      constraintSetup {angularLimitsMax=<0 5 5>, angularLimitsMin=<0 -5 -5>, angularTarget=<0 0 0>}
      localJointOffset=<1 0 -0.2>
     }]
   }
   n12 = Anim Dynamics - Bone: dyn_back_hair_02_01(ComponentPose=@n11, Alpha=1, ExternalForce="0,0,0") {
    angularDampingOverride=0.9999
    angularSpringConstant=75
    bAngularSpring=true
    bGravityOverrideInSimSpace=true
    bOverrideAngularDamping=true
    simulationSpace=RootRelative
    alphaBoolBlend {blendOption=Linear}
    boundBone {boneName=dyn_back_hair_02_01}
    componentLinearAccScale=<1 1 1>
    componentLinearVelScale=<1 1 1>
    physicsBodyDefinitions [{
      sphereCollisionRadius=0
      boundBone {boneName=dyn_back_hair_02_01}
      boxExtents=<10 20 20>
      constraintSetup {angularLimitsMax=<0 8 8>, angularLimitsMin=<0 -8 -8>, angularTarget=<0 0 0>}
      localJointOffset=<1 0 -0.2>
     }]
   }
   n13 = Anim Dynamics - Bone: dyn_right_hair_02_01(ComponentPose=@n12, Alpha=1, ExternalForce="0,0,0") {
    angularDampingOverride=0.9999
    angularSpringConstant=75
    bAngularSpring=true
    bGravityOverrideInSimSpace=true
    bOverrideAngularDamping=true
    simulationSpace=RootRelative
    alphaBoolBlend {blendOption=Linear}
    boundBone {boneName=dyn_right_hair_02_01}
    componentLinearAccScale=<1 1 1>
    componentLinearVelScale=<1 1 1>
    physicsBodyDefinitions [{
      sphereCollisionRadius=0
      boundBone {boneName=dyn_right_hair_02_01}
      boxExtents=<10 20 20>
      constraintSetup {angularLimitsMax=<0 7 7>, angularLimitsMin=<0 -7 -7>, angularTarget=<0 0 0>}
      localJointOffset=<1 0 -0.2>
     }]
   }
   n14 = Anim Dynamics - Bone: dyn_right_hair_04_01(ComponentPose=@n13, Alpha=1, ExternalForce="0,0,0") {
    angularDampingOverride=0.9999
    angularSpringConstant=75
    bAngularSpring=true
    bGravityOverrideInSimSpace=true
    bOverrideAngularDamping=true
    simulationSpace=RootRelative
    alphaBoolBlend {blendOption=Linear}
    boundBone {boneName=dyn_right_hair_04_01}
    componentLinearAccScale=<1 1 1>
    componentLinearVelScale=<1 1 1>
    physicsBodyDefinitions [{
      sphereCollisionRadius=0
      boundBone {boneName=dyn_right_hair_04_01}
      boxExtents=<10 20 20>
      constraintSetup {angularLimitsMax=<0 8 8>, angularLimitsMin=<0 -8 -8>, angularTarget=<0 0 0>}
      localJointOffset=<1 0 -0.2>
     }]
   }
   n15 = Anim Dynamics - Bone: dyn_left_hair_02_01(ComponentPose=@n14, Alpha=1, ExternalForce="0,0,0") {
    angularDampingOverride=0.9999
    angularSpringConstant=75
    bAngularSpring=true
    bGravityOverrideInSimSpace=true
    bOverrideAngularDamping=true
    simulationSpace=RootRelative
    alphaBoolBlend {blendOption=Linear}
    boundBone {boneName=dyn_left_hair_02_01}
    componentLinearAccScale=<1 1 1>
    componentLinearVelScale=<1 1 1>
    physicsBodyDefinitions [{
      sphereCollisionRadius=0
      boundBone {boneName=dyn_left_hair_02_01}
      boxExtents=<10 20 20>
      constraintSetup {angularLimitsMax=<10 10 10>, angularLimitsMin=<-10 -10 -10>, angularTarget=<0 0 0>}
      localJointOffset=<1 0 -0.2>
     }]
   }
   n16 = Anim Dynamics - Bone: dyn_left_hair_01(ComponentPose=@n15, Alpha=1, ExternalForce="0,0,0") {
    angularDampingOverride=0.9999
    angularSpringConstant=75
    bAngularSpring=true
    bGravityOverrideInSimSpace=true
    bOverrideAngularDamping=true
    simulationSpace=RootRelative
    alphaBoolBlend {blendOption=Linear}
    boundBone {boneName=dyn_left_hair_01}
    componentLinearAccScale=<1 1 1>
    componentLinearVelScale=<1 1 1>
    physicsBodyDefinitions [{
      sphereCollisionRadius=0
      boundBone {boneName=dyn_left_hair_01}
      boxExtents=<10 20 20>
      constraintSetup {angularLimitsMax=<8 8 8>, angularLimitsMin=<-8 -8 -8>, angularTarget=<0 0 0>}
      localJointOffset=<1 0 -0.2>
     }]
   }
   n17 = Anim Dynamics - Bone: dyn_top_hair_01(ComponentPose=@n16, Alpha=1, ExternalForce="0,0,0") {
    angularDampingOverride=0.9999
    angularSpringConstant=75
    bAngularSpring=true
    bGravityOverrideInSimSpace=true
    bOverrideAngularDamping=true
    simulationSpace=RootRelative
    alphaBoolBlend {blendOption=Linear}
    boundBone {boneName=dyn_top_hair_01}
    componentLinearAccScale=<1 1 1>
    componentLinearVelScale=<1 1 1>
    physicsBodyDefinitions [{
      sphereCollisionRadius=0
      boundBone {boneName=dyn_top_hair_01}
      boxExtents=<10 20 20>
      constraintSetup {angularLimitsMax=<5 5 5>, angularLimitsMin=<-5 -5 -5>, angularTarget=<0 0 0>}
      localJointOffset=<1 0 -0.2>
     }]
   }
   n18 = Anim Dynamics - Bone: dyn_top_hair_02(ComponentPose=@n17, Alpha=1, ExternalForce="0,0,0") {
    angularDampingOverride=0.9999
    angularSpringConstant=75
    bAngularSpring=true
    bGravityOverrideInSimSpace=true
    bOverrideAngularDamping=true
    simulationSpace=RootRelative
    alphaBoolBlend {blendOption=Linear}
    boundBone {boneName=dyn_top_hair_02}
    componentLinearAccScale=<1 1 1>
    componentLinearVelScale=<1 1 1>
    physicsBodyDefinitions [{
      sphereCollisionRadius=0
      boundBone {boneName=dyn_top_hair_02}
      boxExtents=<10 20 20>
      constraintSetup {angularLimitsMax=<3 3 5>, angularLimitsMin=<-3 -3 -5>, angularTarget=<0 0 0>}
      localJointOffset=<1 0 -0.2>
     }]
   }
   n19 = Anim Dynamics - Bone: dyn_top_hair_03(ComponentPose=@n18, Alpha=1, ExternalForce="0,0,0") {
    angularDampingOverride=0.9999
    angularSpringConstant=75
    bAngularSpring=true
    bGravityOverrideInSimSpace=true
    bOverrideAngularDamping=true
    simulationSpace=RootRelative
    alphaBoolBlend {blendOption=Linear}
    boundBone {boneName=dyn_top_hair_03}
    componentLinearAccScale=<1 1 1>
    componentLinearVelScale=<1 1 1>
    physicsBodyDefinitions [{
      sphereCollisionRadius=0
      boundBone {boneName=dyn_top_hair_03}
      boxExtents=<10 20 20>
      constraintSetup {angularLimitsMax=<6 6 7>, angularLimitsMin=<-6 -6 -7>, angularTarget=<0 0 0>}
      localJointOffset=<1 0 -0.2>
     }]
   }
   Anim Dynamics - Bone: dyn_top_hair_04(ComponentPose=@n19, Alpha=1, ExternalForce="0,0,0") {
    angularDampingOverride=0.9999
    angularSpringConstant=75
    bAngularSpring=true
    bGravityOverrideInSimSpace=true
    bOverrideAngularDamping=true
    simulationSpace=RootRelative
    alphaBoolBlend {blendOption=Linear}
    boundBone {boneName=dyn_top_hair_04}
    componentLinearAccScale=<1 1 1>
    componentLinearVelScale=<1 1 1>
    physicsBodyDefinitions [{
      sphereCollisionRadius=0
      boundBone {boneName=dyn_top_hair_04}
      boxExtents=<10 20 20>
      constraintSetup {angularLimitsMax=<7 7 7>, angularLimitsMin=<-7 -7 -7>, angularTarget=<0 0 0>}
      localJointOffset=<1 0 -0.2>
     }]
   }
   n20 = RigidBody(ComponentPose=Local To Component(LocalPose=Copy Pose From Mesh() {bCopyCurves=true, bUseAttachedParent=true})) {
    bClampLinearTranslationLimitToRefPose=true
    bOverrideWorldGravity=true
    overridePhysicsAsset=&/F_MED_Pencil_Mango_Hair_Physics
    simulationSpace=BaseBoneSpace
    alphaBoolBlend {blendOption=Linear}
    baseBoneRef {boneName=spine_05}
    componentLinearAccScale=<0.15 0.15 0.15>
   }
   Blend Poses by bool(
    BlendPose_0=Component To Local(ComponentPose=@n20)
    BlendTime_0=0.3
    BlendTime_1=0.3
    bActiveValue=OR Boolean(A=Get bIsJumping(), B=Get bIsFalling())
   ) {blendType=Linear}
   Copy Pose From Mesh() {bCopyCurves=true, bUseAttachedParent=true}
   n21 = "Transform (Modify) Bone - Bone: dyn_back_bottom_MSTR"!(
    ComponentPose
    Alpha="To Float (Boolean)"(InBool=Get Head_up() -> (Head_up:bool))
    Translation="-1.003,-2.943,-2.041"
    Rotation="40.146,62.410,63.827"
   ) -> (Pose) {
    rotationMode=BMM_Additive
    rotationSpace=BCS_BoneSpace
    translationMode=BMM_Additive
    translationSpace=BCS_BoneSpace
    alphaBoolBlend {blendOption=Linear}
    alphaScaleBiasClamp {bInterpResult=true, interpSpeedDecreasing=5, interpSpeedIncreasing=7}
    boneToModify {boneName=dyn_back_bottom_MSTR}
    rotation=<rot 40.146 62.4099 63.827>
    translation=<-1.003 -2.943 -2.041>
   }
   "Transform (Modify) Bone - Bone: dyn_back_bottom_MSTR"(ComponentPose=@n21, Alpha=Get Head Down() -> (Head Down:double), Translation="0,-1.3,1.5", Rotation="0,-10,0") {
    rotationMode=BMM_Replace
    rotationSpace=BCS_BoneSpace
    scaleMode=BMM_Replace
    scaleSpace=BCS_BoneSpace
    translationMode=BMM_Replace
    translationSpace=BCS_BoneSpace
    alphaBoolBlend {blendOption=Linear}
    alphaScaleBiasClamp {bInterpResult=true, interpSpeedDecreasing=5, interpSpeedIncreasing=7}
    boneToModify {boneName=dyn_back_bottom_MSTR}
    rotation {yaw=-10}
    translation {y=-1.3, z=1.5}
   }
   n22 = Control Rig(Source=Input Pose()) {
    alphaBoolBlend {blendOption=Linear}
    controlRigAssetReference {blueprintRigClass=&/F_MED_Pencil_Mango_Hair_CtrlRig.F_MED_Pencil_Mango_Hair_CtrlRig_C}
   }
   n23 = RigidBody(ComponentPose=Local To Component(LocalPose=@n22)) {
    bClampLinearTranslationLimitToRefPose=true
    bOverrideWorldGravity=true
    overridePhysicsAsset=&F_MED_Pencil_Mango_Hair_Physics
    simulationSpace=BaseBoneSpace
    alphaBoolBlend {blendOption=Linear}
    baseBoneRef {boneName=head}
    componentAppliedLinearAccClamp=<950 950 950>
    componentLinearAccScale=<0.4 0.4 0.5>
    overrideWorldGravity {z=-700}
    simSpaceSettings {maxAngularAcceleration=1000, maxAngularVelocity=3, maxLinearAcceleration=250, maxLinearVelocity=500, velocityScaleZ=0, worldAlpha=1}
   }
   n24 = "Transform (Modify) Bone - Bone: dyn_back_bottom_01"(ComponentPose=@n23, Alpha=1, Translation="0,2,0", Rotation="0,0,0", Scale="1,1,1") {
    rotationMode=BMM_Additive
    rotationSpace=BCS_BoneSpace
    scaleMode=BMM_Additive
    scaleSpace=BCS_BoneSpace
    translationMode=BMM_Additive
    translationSpace=BCS_BoneSpace
    alphaBoolBlend {blendOption=Linear}
    boneToModify {boneName=dyn_back_bottom_01}
    translation {y=2}
   }
   Output Pose(Result=Component To Local(ComponentPose=@n24)) {layerGroup=None}
  ]}
 EventGraph {nodes [
   Event BlueprintUpdateAnimation()
   > Set Head_up(Head_up:bool=true) -> (Output_Get:bool)
   n1 = GetOwnerSkeletalMeshComponent()
   n2 = BreakRotator(InRot=BreakTransform(InTransform=GetDeltaTransformFromRefPose(self=@n1, BoneName=head, BaseName=spine_05)).Rotation)
   n3 = AND Boolean(A=Get bIsCrouching(), B=Get bIsMoving2D())
   SelectFloat(A=MapRangeClamped(Value=@n2.Yaw, InRangeA=-3, InRangeB=-40), B=0, bPickA=AND Boolean(A=Get bIsTargeting(), B=NOT Boolean(A=@n3)))
   Set Head Down(Head Down:double=SelectFloat(A=MapRangeClamped(Value=@n2.Yaw, InRangeA=1, InRangeB=10), B=0, bPickA=Get bIsTargeting())) -> (Output_Get:double)
  ]}
]
variables [
 Head Down {type=number}
 Head_up {type=boolean}
]
